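# Material for the Viking 2 vehicle wheel cosmetic, shading a rim with metal, wood and stone regions on skeletal meshes.
# Channels of a mask texture drive wood and stone tints over the diffuse map, scale metalness, and sum per-material roughness values.
# The resulting material attributes, including a normal map, pass through the shared vehicle damage function and then the de-rez transition function.
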Material /VehicleCosmetics/Wheels/Viking/Materials/M_Wheel_Viking2
graph {
 properties {bAutomaticallySetUsageInEditor=false, bUseMaterialAttributes=true, bUsedWithSkeletalMesh=true, usedShadingModels=Default Lit}
 nodes [
  n1 = TextureSampleParameter2D() {parameterName=Diffuse, texture=/VehicleCosmetics/Wheels/Viking/Textures/T_Viking_Rim_BC}
  n2 = VectorParameter() {parameterName=WoodColor, defaultValue=<rgba 0.125 0.0273 0 1>}
  n3 = Multiply(A=NamedRerouteUsage()) {constB=0.7}
  n4 = NamedRerouteUsage()
  n5 = Multiply(A=NamedRerouteUsage(), B=ScalarParameter() {defaultValue=0.2, parameterName=MetalRoughness})
  n6 = TextureSampleParameter2D() {parameterName=Mask, samplerType=SAMPLERTYPE_LinearColor, texture=/VehicleCosmetics/Wheels/Viking/Textures/T_Viking_Rim_M}
  n7 = Multiply(
   A=ScalarParameter() {defaultValue=0.8, parameterName=WoodRoughness}
   B=NamedRerouteDeclaration(None=@n6.G) {name=G, nodeColor=<rgba 0.3359 0 1 1>}
  )
  n8 = NamedRerouteUsage()
  n9 = ScalarParameter() {defaultValue=0.4, parameterName=StoneRoughness}
  n10 = SetMaterialAttributes!(
   MaterialAttributes
   Base Color=LinearInterpolate(
    A=LinearInterpolate(A=@n1.RGB, B=@n2.RGB, Alpha=@n3)
    B=VectorParameter() {parameterName=StoneColor, defaultValue=<rgba 0.6947 0.102 0.8594 1>}.RGB
    Alpha=Multiply(A=@n4) {constB=0.7}
   )
   Emissive Color
   Normal=TextureSampleParameter2D() {parameterName=Normal, samplerType=SAMPLERTYPE_Normal, texture=/VehicleCosmetics/Wheels/Viking/Textures/T_Viking_Rim_N}.RGB
   Metallic=Multiply(A=NamedRerouteUsage(), B=ScalarParameter() {defaultValue=1, parameterName=Metallness})
   Roughness=Add(A=Add(A=@n5, B=@n7), B=Multiply(A=@n8, B=@n9))
  ) {attributeSetTypes [
    #1
    #2
    #3
    #4
    #5
   ]}
  MaterialFunctionCall(In=MaterialFunctionCall(Material=@n10) {materialFunction=&MF_BaseVehicle_Damage}.Result) {materialFunction=&MF_VehicleDeRez_Transition}
  Multiply(B=ScalarParameter() {defaultValue=0.4, parameterName=StoneRoughness})
  NamedRerouteDeclaration(None=@n6.R) {name=R, nodeColor=<rgba 1 0.6562 0 1>}
  NamedRerouteDeclaration(None=@n6.B) {name=B, nodeColor=<rgba 0 1 0.0156 1>}
 ]
}
